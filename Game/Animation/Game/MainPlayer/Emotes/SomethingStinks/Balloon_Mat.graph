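# Subsurface-shaded, glossy red-orange balloon material for the main player's SomethingStinks emote, enabled for skeletal meshes, mesh deformers and Nanite.
# A parameterized Fresnel blends subsurface color from a darkened base to full color, and opacity between inner and outer strength parameters.
# A quality switch passes these attributes through, except on Low, which zeroes subsurface color and opacity.
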Material /Game/Animation/Game/MainPlayer/Emotes/SomethingStinks/Balloon_Mat
graph {
 properties {
  bUseMaterialAttributes=true
  bUsedWithMeshDeformer=true
  bUsedWithNanite=true
  bUsedWithSkeletalMesh=true
  pixelDepthOffsetMode=PDOM_Legacy
  shadingModel=MSM_Subsurface
  usedShadingModels=Subsurface
 }
 nodes [
  n1 = Constant3Vector() {constant=<rgba 1 0.1996 0 0>}
  n2 = Fresnel(ExponentIn=ScalarParameter() {defaultValue=1, group=SS Shader Model, parameterName=SS Fresnel Exponent})
  n3 = Multiply(A=@n1, B=ScalarParameter() {defaultValue=0.5, parameterName=balloon ss color modifier})
  n4 = MakeMaterialAttributes!(
   BaseColor=@n1
   Metallic
   Specular=Constant() {r=0.85}
   Roughness=Constant() {r=0.3}
   Anisotropy
   EmissiveColor
   Opacity=LinearInterpolate(
    A=ScalarParameter() {defaultValue=0.05, group=SS Shader Model, parameterName=SS Inner Strength}
    B=ScalarParameter() {defaultValue=1.3801, group=SS Shader Model, parameterName=SS Outer Strength}
    Alpha=@n2
   )
   OpacityMask
   Normal
   Tangent
   WorldPositionOffset
   SubsurfaceColor=LinearInterpolate(A=Saturate(None=@n3), B=@n1, Alpha=@n2)
   ClearCoat
   ClearCoatRoughness
   AmbientOcclusion
   Refraction
   CustomizedUVs_0
   CustomizedUVs_1
   CustomizedUVs_2
   CustomizedUVs_3
   CustomizedUVs_4
   CustomizedUVs_5
   CustomizedUVs_6
   CustomizedUVs_7
   PixelDepthOffset
   ShadingModel
   Displacement
  )
  n5 = Constant()
  QualitySwitch(
   Default=@n4
   Low=SetMaterialAttributes!(MaterialAttributes=@n4, Subsurface Color=@n5, Opacity=@n5) {attributeSetTypes [#1, #2]}
   High=@n4
   Medium=@n4
  )
 ]
}
