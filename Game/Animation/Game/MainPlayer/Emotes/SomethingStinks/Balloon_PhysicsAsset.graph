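# Physics asset for the temporary balloon prop in the main player's SomethingStinks emote.
# It gives the balloon bone and its tip bone one small capsule body each, both of equal size.
# A single constraint joins the tip to the balloon, with matching angle limits set for both swing axes and twist.
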PhysicsAsset /Game/Animation/Game/MainPlayer/Emotes/SomethingStinks/Balloon_PhysicsAsset
bodies {
 BalloonTempProp [{shapeType=Capsule, length=0.505, radius=0.505, rotation=<rot 0 -0.0097 -89.995>}]
 BalloonTipTempProp [{shapeType=Capsule, length=0.505, radius=0.505, center=<-0.0381 0.0325 0.0648>, rotation=<rot 0 -0.8843 -88.4966>}]
}
constraints [{bone1Name=BalloonTipTempProp, bone2Name=BalloonTempProp, swing1LimitDegrees=45, swing2LimitDegrees=45, twistLimitDegrees=45}]
